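# Translucent, unlit, two-sided Niagara ribbon material for the swipe trail of the Lyrical female pickaxe.
# Dynamic parameters offset the trail texture in U and mirror it by swing direction.
# The trail texture, modulated by a panning contrail, blends two dark paint colors, optionally replaced by particle color.
# Opacity combines trail and contrail alpha, panning-noise erosion thresholded by particle alpha, a U-end fade and depth fade (skipped on mobile).
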Material /BRCosmetics/Weapons/FORT_Melee/Pickaxe_Lyrical_Female/FX/M_LyricalSwipeRibbon
graph {
 properties {
  bUsedWithNiagaraRibbons=true
  blendMode=BLEND_Translucent
  pixelDepthOffsetMode=PDOM_Legacy
  shadingModel=MSM_Unlit
  twoSided=true
  usedShadingModels=Unlit
 }
 nodes [
  n1 = VectorParameter() {parameterName=PaintColorHigh, defaultValue=<rgba 0.008 0.008 0.02 0>}
  n2 = VectorParameter() {parameterName=PaintColorLow, defaultValue=<rgba 0.003 0.003 0.01 0>}
  n3 = MaterialFunctionCall(
   UVs=Saturate(None=TextureCoordinate())
   Texture Scale=AppendVector(A=ScalarParameter() {defaultValue=1, parameterName=TexScaleU}, B=ScalarParameter() {defaultValue=1, parameterName=TexScaleV})
  ) {materialFunction=&ScaleUVsByCenter}
  n4 = MaterialFunctionCall(Float2=Multiply(A=@n3.UVs, B=Constant2Vector() {g=-1, r=1})) {materialFunction=&BreakOutFloat2Components}
  n5 = AppendVector(A=OneMinus(None=@n4.R), B=@n4.G)
  n6 = DynamicParameter() {paramNames [UOffset,SwingRight,Param3,Param4]}
  n7 = DynamicParameter() {paramNames [UOffset,SwingRight,Param3,Param4]}
  n8 = Constant()
  n9 = TextureSampleParameter2D(UVs=Add(A=LinearInterpolate(A=@n3.UVs, B=@n5, Alpha=@n6.SwingRight), B=AppendVector(A=@n7.UOffset, B=@n8))) {
   parameterName=PrimaryTexture
   samplerType=SAMPLERTYPE_LinearColor
   texture=&/T_Lyrical_Trail
  }
  n10 = ChannelMaskParameter(None=@n9.RGBA) {maskChannel=Green, parameterName=ColorChannel, defaultValue=<rgba 0 1 0 0>}
  n11 = TextureSample(UVs=Panner(Coordinate=TextureCoordinate() {coordinateIndex=1, uTiling=0.8}) {speedX=1.5}) {texture=/Game/Effects/Fort_Effects/Textures/T_Contrail_Lava}
  n12 = Multiply(A=@n10.R, B=@n11.B)
  n13 = ParticleColor()
  MaterialFunctionCall(Emissive=LinearInterpolate(
    A=LinearInterpolate(A=@n1.RGB, B=@n2.RGB, Alpha=@n12)
    B=Multiply(A=@n13.RGB, B=@n10.R)
    Alpha=ScalarParameter() {parameterName=Use Particle Color}
   )) {materialFunction=&MF_FX_EmissiveAdjustments}
  n14 = Panner(Coordinate=Multiply(A=TextureCoordinate() {coordinateIndex=1, uTiling=0.5}, B=ScalarParameter() {defaultValue=2, parameterName=EatawayTile})) {speedX=1}
  n15 = Constant()
  n16 = Saturate(None=If(
    A=TextureSample(UVs=@n14) {texture=/Game/Effects/Fort_Effects/Textures/Tile/T_Smooth_Tile_Chromatic}.G
    B=@n13.A
    "A > B"=@n15
    "A == B"=@n15
    "A < B"=Constant() {r=1}
   ))
  n17 = Multiply(A=ChannelMaskParameter(None=@n9.RGBA) {maskChannel=Alpha, parameterName=AlphaChannel, defaultValue=<rgba 0 0 0 1>}.R, B=@n11.A)
  n18 = TextureCoordinate()
  n19 = Multiply(A=OneMinus(None=ComponentMask(None=@n18) {r=true})) {constB=4}
  n20 = ScalarParameter() {defaultValue=20, group=DepthDade, parameterName=DepthFadeDistance, sortPriority=1}
  Multiply(
   A=Multiply(A=Multiply(A=@n16, B=@n17) {constB=0.15}, B=Saturate(None=@n19)) {constB=0.15}
   B=ShadingPathSwitch(Default=DepthFade(FadeDistance=@n20), Mobile=Constant() {r=1}) {desc=Migrated from FeatureLevelSwitch}
  )
 ]
}
