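# Physics asset for the Lyrical Female pickaxe, a two-handed melee harvesting tool cosmetic.
# A long capsule covers the handle bone, and two shorter, thicker capsules cover the dynamic can joints.
# Each can joint is constrained directly to the handle with tight swing and twist limits, allowing only slight movement relative to it.
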PhysicsAsset /BRCosmetics/Weapons/FORT_Melee/Pickaxe_Lyrical_Female/Meshes/Lyrical_Female_Axe_Physics
bodies {
 Melee_TwoHanded_Handle [{shapeType=Capsule, length=100.3642, radius=3.1364, center=<1.0313 -1.3954 19.1094>, rotation=<rot -0.2124 -90.0024 0.1914>}]
 dyn_can_joint1 [{
   shapeType=Capsule
   length=34.25
   radius=6.5
   shapeName=dyn_can_joint1_capsule
   center=<0.1038 -0.0627 0.0642>
   rotation=<rot 0.0076 -1.4609 -89.6774>
  }]
 dyn_can_joint2 [{
   shapeType=Capsule
   length=35.25
   radius=6.25
   shapeName=dyn_can_joint2_capsule
   center=<-0.0189 0.0276 -0.876>
   rotation=<rot -1.2921 -1.7891 -99.1405>
  }]
}
constraints [
 {bone1Name=dyn_can_joint2, bone2Name=Melee_TwoHanded_Handle, swing1LimitDegrees=8, swing2LimitDegrees=8, twistLimitDegrees=14}
 {bone1Name=dyn_can_joint1, bone2Name=Melee_TwoHanded_Handle, swing1LimitDegrees=8, swing2LimitDegrees=8, twistLimitDegrees=14}
]
